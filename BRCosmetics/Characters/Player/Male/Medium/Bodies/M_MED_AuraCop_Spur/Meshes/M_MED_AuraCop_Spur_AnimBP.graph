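# Post-process style AnimBlueprint for the male medium AuraCop Spur body cosmetic, adding secondary motion to its left and right guns.
# It copies the attached parent mesh's pose, runs the body's Control Rig, and constrains each gun base to weighted torso bones.
# World-space AnimDynamics with angular springs and limits then sway both guns.
# Events set gravity-override and front-end-dependent gun alpha/sim-space variables, which the AnimGraph does not visibly read.
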AnimBlueprint /BRCosmetics/Characters/Player/Male/Medium/Bodies/M_MED_AuraCop_Spur/Meshes/M_MED_AuraCop_Spur_AnimBP
target_skeleton=&/M_MED_AuraCop_Spur_Skeleton
graphs [
 AnimGraph {nodes [
   n1 = Control Rig(Source=Copy Pose From Mesh() {bUseAttachedParent=true}) {
    alphaBoolBlend {blendOption=Linear}
    controlRigAssetReference {blueprintRigClass=&/M_MED_AuraCop_Spur_CtrlRig.M_MED_AuraCop_Spur_CtrlRig_C}
   }
   n2 = "Constraint - base_gun_r (PPPPPPP)"!(
    ComponentPose=Local To Component(LocalPose=@n1)
    Alpha=1
    ConstraintWeights_0=0.3149
    ConstraintWeights_1:float=0.2496
    ConstraintWeights_2:float=0.1356
    ConstraintWeights_3:float=0.1131
    ConstraintWeights_4:float=0.1077
    ConstraintWeights_5:float=0.0502
    ConstraintWeights_6:float=0.029
   ) -> (Pose) {
    alphaBoolBlend {blendOption=Linear}
    boneToModify {boneName=base_gun_r}
    constraintSetup [
     {targetBone {boneName=spine_03}}
     {targetBone {boneName=spine_02}}
     {targetBone {boneName=spine_04}}
     {targetBone {boneName=spine_01}}
     {targetBone {boneName=clavicle_l}}
     {targetBone {boneName=pelvis}}
     {targetBone {boneName=deform_pec_l}}
    ]
    constraintWeights [1,1,1,1,1,1,1]
   }
   n3 = "Constraint - base_gun_l (PPPPPPP)"!(
    ComponentPose=@n2
    Alpha=1
    ConstraintWeights_0=0.3276
    ConstraintWeights_1:float=0.2339
    ConstraintWeights_2:float=0.1397
    ConstraintWeights_3:float=0.1054
    ConstraintWeights_4:float=0.1011
    ConstraintWeights_5:float=0.0522
    ConstraintWeights_6:float=0.0401
   ) -> (Pose) {
    alphaBoolBlend {blendOption=Linear}
    boneToModify {boneName=base_gun_l}
    constraintSetup [
     {targetBone {boneName=spine_03}}
     {targetBone {boneName=spine_02}}
     {targetBone {boneName=spine_04}}
     {targetBone {boneName=spine_01}}
     {targetBone {boneName=clavicle_r}}
     {targetBone {boneName=deform_pec_r}}
     {targetBone {boneName=pelvis}}
    ]
    constraintWeights [1,1,1,1,1,1,1]
   }
   n4 = Anim Dynamics - Bone: dyn_gun_l!(SimulationSpace:AnimPhysSimSpaceType Enum=World, ComponentPose=@n3, Alpha=1, ExternalForce="0,0,0") -> (Pose) {
    angularDampingOverride=1
    angularSpringConstant=20000
    bAngularSpring=true
    bOverrideAngularDamping=true
    bUseGravityOverride=true
    gravityScale=0.35
    simulationSpace=World
    alphaBoolBlend {blendOption=Linear}
    boundBone {boneName=dyn_gun_l}
    componentLinearAccScale=<1 1 1>
    componentLinearVelScale=<1 1 1>
    gravityOverride {z=-980}
    physicsBodyDefinitions [{
      boundBone {boneName=dyn_gun_l}
      boxExtents=<50 50 50>
      constraintSetup {angularLimitsMax=<10 3 1>, angularLimitsMin=<-10 -3 -7>, angularTarget=<1 0 0>}
      localJointOffset=<2 0 0>
     }]
   }
   n5 = Anim Dynamics - Bone: dyn_gun_r!(SimulationSpace:AnimPhysSimSpaceType Enum=World, ComponentPose=@n4, Alpha=1, ExternalForce="0,0,0") -> (Pose) {
    angularDampingOverride=1
    angularSpringConstant=20000
    bAngularSpring=true
    bOverrideAngularDamping=true
    bUseGravityOverride=true
    gravityScale=0.35
    simulationSpace=World
    alphaBoolBlend {blendOption=Linear}
    boundBone {boneName=dyn_gun_r}
    componentLinearAccScale=<1 1 1>
    componentLinearVelScale=<1 1 1>
    gravityOverride {z=-980}
    physicsBodyDefinitions [{
      boundBone {boneName=dyn_gun_r}
      boxExtents=<50 50 50>
      constraintSetup {angularLimitsMax=<10 3 1>, angularLimitsMin=<-10 -3 -7>, angularTarget=<1 0 0>}
      localJointOffset=<2 0 0>
     }]
   }
   Output Pose(Result=Component To Local(ComponentPose=@n5)) {layerGroup=None}
  ]}
 EventGraph {nodes [
   Event BlueprintUpdateAnimation()
   > n1 = Gravity Override State Function DEPRECATED(CharacterPartAnimInstance=Self-Reference(), GameTimeInSeconds=GetGameTimeInSeconds(), GravityOverrideParams=Get Gravity Override Params())
   > Set GravityOverride(GravityOverride=@n1)
   n2 = Event OnCharacterCustomizationLoaded()
   > Set Skel_Mesh(Skel_Mesh=@n2.CustomBaseMesh)
   > n3 = Branch(Condition=Get bIsInFrontEnd())
   > Set Guns Sim Space(Guns Sim Space:AnimPhysSimSpaceType Enum=World) -> (Output_Get:AnimPhysSimSpaceType Enum)
   > Set Guns Alpha(Guns Alpha:float=0.4) -> (Output_Get:float)
   Set Guns Sim Space(execute=@n3.else)
   > Set Guns Alpha(Guns Alpha=1)
  ]}
]
variables [
 Gravity Override Params {type=GravityOverrideParamsStruct}
 GravityOverride {type=Vector}
 Guns Alpha {type=number}
 Guns Sim Space {type=AnimPhysSimSpaceType}
 Skel_Mesh {type=SkeletalMeshComponent}
]
